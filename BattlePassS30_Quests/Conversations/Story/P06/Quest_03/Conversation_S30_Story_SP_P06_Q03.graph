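# Conversation graph for Season 30 story quest P06 Q03, where Magneto warns the player about the box and the Wanderer.
# One entry tag leads into a linear five-line Speech chain advanced by Continue choices.
# The auto-selected first choice requires the RoundThumb NPC conversation tag, the quest active and objective obj0 incomplete via an inverted check.
# The second line's choice carries a quest-specific gameplay tag.
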ConversationDatabase /BattlePassS30_Quests/Conversations/Story/P06/Quest_03/Conversation_S30_Story_SP_P06_Q03
entry_points [{tag=Conversation.Entry.Season30.Story.P06.Q03.MX, destinations [#1]}]
nodes [
 ConversationEntryPointNode_0 {
  id=#1
  class=ConversationEntryPointNode
  outputs [#2]
  properties {entryTag {tagName=Conversation.Entry.Season30.Story.P06.Q03.MX}}
 }
 Speech_C_5 {
  id=#2
  class=Speech_C
  outputs [#3]
  properties {softTaskIcon=&T_UI_NPC_MainQuest, generalConfig {defaultMessage="You're the keepers of that box, are you? "}}
  sub_nodes [
   ConversationChoiceNode_0 {
    class=ConversationChoiceNode
    properties {
     defaultChoiceDisplayText=Talk to Magneto
     choiceTags {gameplayTags [{tagName=Conversation.Choice.AutoSelectIfValid}, {tagName=Conversation.Choice.Category.MainQuestline}]}
    }
   }
   FortConversationRequirement_ControllerRequirement_0 {
    class=FortConversationRequirement_ControllerRequirement
    properties {
     nodeName=Require: NPC Has Conversation Tag
     participantId {tagName=Conversation.Participant.NPC}
     requirement {class=FortControllerRequirement_NPCHasConversationEntryTag, properties {queryTag {tagName=Conversation.Entry.RoundThumb}}}
    }
   }
   FortConversationRequirement_ControllerRequirement_1 {
    class=FortConversationRequirement_ControllerRequirement
    properties {
     nodeName=Require: Is Quest Active
     requirement {
      class=FortControllerRequirement_IsQuestActive
      properties {questDefinition=/BattlePassS30_Quests/Items/QuestItems/Story/P06/Quest_S30_Story_P06_Q03}
     }
    }
   }
   FortConversationRequirement_ControllerRequirement_2 {
    class=FortConversationRequirement_ControllerRequirement
    properties {
     nodeName=Require: Has Completed Quest Objective
     requirement {
      class=FortControllerRequirement_HasCompletedQuestObjective
      properties {
       bInverseCondition=true
       objectiveBackendName=Quest_S30_Story_P06_Q03_obj0
       questDefinition=&Quest_S30_Story_P06_Q03
      }
     }
    }
   }
  ]
 }
 Speech_C_2 {
  id=#3
  class=Speech_C
  outputs [#4]
  properties {generalConfig {defaultMessage="Forgive me for speaking plainly, but that thing's going to destroy us all."}}
  sub_nodes [ConversationChoiceNode_3 {
    class=ConversationChoiceNode
    properties {defaultChoiceDisplayText=Continue, choiceTags {gameplayTags [{tagName=Athena.Quests.S30.Conversation.Story.P06.03.MX}]}}
   }]
 }
 Speech_C_0 {
  id=#4
  class=Speech_C
  outputs [#5]
  properties {generalConfig {defaultMessage="I've already set up beacons to call in reinforcements for you, but there's only so much we can do."}}
  sub_nodes [ConversationChoiceNode_6 {class=ConversationChoiceNode, properties {defaultChoiceDisplayText=Continue}}]
 }
 Speech_C_3 {
  id=#5
  class=Speech_C
  outputs [#6]
  properties {generalConfig {defaultMessage="I can't verify this Wanderer is who I think he is, but if it's him... If he's the one seizing that box's powers..."}}
  sub_nodes [ConversationChoiceNode_5 {class=ConversationChoiceNode, properties {defaultChoiceDisplayText=Continue}}]
 }
 Speech_C_4 {
  id=#6
  class=Speech_C
  outputs []
  properties {generalConfig {defaultMessage="This place may be doomed. "}}
  sub_nodes [ConversationChoiceNode_4 {class=ConversationChoiceNode, properties {defaultChoiceDisplayText=Continue}}]
 }
]
